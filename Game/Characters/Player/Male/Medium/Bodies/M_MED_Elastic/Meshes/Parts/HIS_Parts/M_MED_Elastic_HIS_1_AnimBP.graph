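# Animation Blueprint for the M_MED_Elastic_HIS_1 player character part, adding secondary physics motion to its hair bones.
# It copies the attached parent mesh's pose, then chains five single-bone Anim Dynamics nodes with angular springs, damping, reduced gravity and per-bone angular limits in root-relative space.
# Only two right-side bones take effect, one at half and one at full alpha; the other three have zero alpha.
# The event graph has no logic.
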AnimBlueprint /Game/Characters/Player/Male/Medium/Bodies/M_MED_Elastic/Meshes/Parts/HIS_Parts/M_MED_Elastic_HIS_1_AnimBP
target_skeleton=&/M_MED_Elastic_HIS_1_Skeleton
graphs [
 AnimGraph {nodes [
   n1 = Anim Dynamics - Bone: dyn_hairHIS1_D_01_l(ComponentPose=Local To Component(LocalPose=Copy Pose From Mesh() {bUseAttachedParent=true}), Alpha=0, ExternalForce="0,0,0") {
    angularDampingOverride=1
    angularSpringConstant=4000
    bAngularSpring=true
    bGravityOverrideInSimSpace=true
    bOverrideAngularDamping=true
    gravityScale=0.35
    simulationSpace=RootRelative
    alphaBoolBlend {blendOption=Linear}
    boundBone {boneName=dyn_hairHIS1_D_01_l}
    componentLinearAccScale=<0.3 0.3 0.6>
    componentLinearVelScale=<0.3 0.3 0.6>
    physicsBodyDefinitions [{
      sphereCollisionRadius=0
      boundBone {boneName=dyn_hairHIS1_D_01_l}
      boxExtents=<10 20 20>
      constraintSetup {angularLimitsMax=<0 40 0>, angularLimitsMin=<0 -40 0>, angularTarget=<1 0 0>}
      localJointOffset=<5 0 0>
     }]
   }
   n2 = Anim Dynamics - Bone: dyn_hairHIS1_C_01_l(ComponentPose=@n1, Alpha=0, ExternalForce="0,0,0") {
    angularDampingOverride=1
    angularSpringConstant=4000
    bAngularSpring=true
    bGravityOverrideInSimSpace=true
    bOverrideAngularDamping=true
    gravityScale=0.35
    simulationSpace=RootRelative
    alphaBoolBlend {blendOption=Linear}
    boundBone {boneName=dyn_hairHIS1_C_01_l}
    componentLinearAccScale=<0.3 0.3 0.6>
    componentLinearVelScale=<0.3 0.3 0.6>
    physicsBodyDefinitions [{
      sphereCollisionRadius=0
      boundBone {boneName=dyn_hairHIS1_C_01_l}
      boxExtents=<10 20 20>
      constraintSetup {angularLimitsMax=<0 40 0>, angularLimitsMin=<0 -40 0>, angularTarget=<1 0 0>}
      localJointOffset=<5 0 0>
     }]
   }
   n3 = Anim Dynamics - Bone: dyn_hairHIS1_B_01(ComponentPose=@n2, Alpha=0, ExternalForce="0,0,0") {
    angularDampingOverride=1
    angularSpringConstant=4000
    bAngularSpring=true
    bGravityOverrideInSimSpace=true
    bOverrideAngularDamping=true
    gravityScale=0.35
    simulationSpace=RootRelative
    alphaBoolBlend {blendOption=Linear}
    boundBone {boneName=dyn_hairHIS1_B_01}
    componentLinearAccScale=<0.3 0.3 0.6>
    componentLinearVelScale=<0.3 0.3 0.6>
    physicsBodyDefinitions [{
      sphereCollisionRadius=0
      boundBone {boneName=dyn_hairHIS1_B_01}
      boxExtents=<10 20 20>
      constraintSetup {angularLimitsMax=<0 2 0>, angularLimitsMin=<0 -60 0>, angularTarget=<1 0 0>}
      localJointOffset=<5 0 0>
     }]
   }
   n4 = Anim Dynamics - Bone: dyn_hairHIS1_B_01_r(ComponentPose=@n3, Alpha=0.5, ExternalForce="0,0,0") {
    angularDampingOverride=1
    angularSpringConstant=4000
    bAngularSpring=true
    bGravityOverrideInSimSpace=true
    bOverrideAngularDamping=true
    gravityScale=0.35
    simulationSpace=RootRelative
    alphaBoolBlend {blendOption=Linear}
    boundBone {boneName=dyn_hairHIS1_B_01_r}
    componentLinearAccScale=<0.3 0.3 0.6>
    componentLinearVelScale=<0.3 0.3 0.6>
    physicsBodyDefinitions [{
      sphereCollisionRadius=0
      boundBone {boneName=dyn_hairHIS1_B_01_r}
      boxExtents=<10 20 20>
      constraintSetup {angularLimitsMax=<0 60 0>, angularLimitsMin=<0 -40 0>, angularTarget=<1 0 0>}
      localJointOffset=<5 0 0>
     }]
   }
   n5 = Anim Dynamics - Bone: dyn_hairHIS1_A_02_r(ComponentPose=@n4, Alpha=1, ExternalForce="0,0,0") {
    angularDampingOverride=1
    angularSpringConstant=4000
    bAngularSpring=true
    bGravityOverrideInSimSpace=true
    bOverrideAngularDamping=true
    gravityScale=0.35
    simulationSpace=RootRelative
    alphaBoolBlend {blendOption=Linear}
    boundBone {boneName=dyn_hairHIS1_A_02_r}
    componentLinearAccScale=<0.3 0.3 0.6>
    componentLinearVelScale=<0.3 0.3 0.6>
    physicsBodyDefinitions [{
      sphereCollisionRadius=0
      boundBone {boneName=dyn_hairHIS1_A_02_r}
      boxExtents=<10 20 20>
      constraintSetup {angularLimitsMax=<0 0 0>, angularLimitsMin=<0 -40 0>, angularTarget=<1 0 0>}
      localJointOffset=<5 0 0>
     }]
   }
   Output Pose(Result=Component To Local(ComponentPose=@n5)) {layerGroup=None}
  ]}
 EventGraph {nodes [Event BlueprintUpdateAnimation()]}
]
variables []
